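# Legacy editor utility widget for VO tooling that pairs a character name label with an editable DialogueVoice asset reference.
# On PreConstruct, a SinglePropertyView targets the widget itself and its DialogueVoice variable, exposing that variable as a property field without custom picker code.
# GetName and SetName read and write the CharacterName text block as strings, so the label itself stores the name.
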EditorUtilityWidgetBlueprint /Game/Sounds/Fort_Audio_Blueprints/Utilities/EditorUtilityWidgets/VO_Utilities/Legacy/EUW_CharacterDialogueVoice
graphs [
 EventGraph {nodes [
   Event PreConstruct()
   n1 = Get DialogueVoiceRef() -> (DialogueVoiceRef:Single Property View Object Reference)
   > SetObject(self=@n1, NewObject=Self-Reference())
   > SetPropertyName!(execute, self:Single Property View Object Reference=@n1, NewPropertyName:Name=DialogueVoice) -> (then)
  ]}
 GetName {nodes [
   GetName() {is=title K2Node_FunctionEntry}
   n1 = Get CharacterName() -> (CharacterName:Common Text Object Reference)
   > Return Node!(execute, Name:String="To String (Text)"(InText="GetText (Text)"!(self:Text Object Reference=@n1) -> (ReturnValue:Text)))
  ]}
 SetName {nodes [
   n1 = SetName!() -> (then, Name:String) {is=title K2Node_FunctionEntry}
   > "SetText (Text)"(self=Get CharacterName(), InText="To Text (String)"(InString=@n1))
  ]}
]
variables [DialogueVoice {type=DialogueVoice, default=null}]
widgets {
 info {rootWidgetClass=&CanvasPanel, widgetCount=3}
 widgets [
  CanvasPanel_0 {widgetClassPath=&CanvasPanel, widget=CanvasPanel_0}
  CharacterName {
   widgetClassPath=&CommonTextBlock
   bIsVariable=true
   parent=CanvasPanel_0
   widget=CharacterName
   properties {
    style=&TextStyle-Base_C
    text=Text Block
    font {
     bForceMonospaced=false
     bMaterialIsStencil=false
     fontMaterial=null
     fontObject=&NotoSans
     letterSpacing=0
     monospacedWidth=1
     size=24
     skewAmount=0
     typefaceFontName=NotoSans-Regular
     outlineSettings {
      bApplyOutlineToDropShadows=false
      bMiteredCorners=false
      bSeparateFillAlpha=false
      outlineMaterial=null
      outlineSize=0
      outlineColor=<rgba 0 0 0 1>
     }
    }
    renderTransform {angle=0, scale=<0.5 0.5>, shear=<0 0>, translation=<0 0>}
    shadowOffset=<0 0>
   }
   slot {
    class=CanvasPanelSlot
    properties {layoutData {alignment=<0 0>, anchors {maximum=<0 0>, minimum=<0 0>}, offsets=<ltrb 0 0 160.0601 36.6066>}}
   }
  }
  DialogueVoiceRef {
   widgetClassPath=&SinglePropertyView
   parent=CanvasPanel_0
   widget=DialogueVoiceRef
   properties {renderTransform {angle=0, scale=<0.9 0.9>, shear=<0 0>, translation=<0 0>}}
   slot {
    class=CanvasPanelSlot
    properties {layoutData {alignment=<0 0>, anchors {maximum=<0 0>, minimum=<0 0>}, offsets=<ltrb 144 -8 465.699 56.6934>}}
   }
  }
 ]
}
